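# Masked, unlit material for the NerdBig attachment in the Nerd vehicle body's cosmetic skins, usable on skeletal meshes and mesh deformers.
# MF_Skin_Endit supplies the look, but only its Emissive Color is carried through.
# The red channel of T_NerdBigAttachment_Alpha provides the Opacity Mask.
# Both are set as material attributes and passed through the MF_Vehicle_FNPostFX function.
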Material /VehicleCosmetics/Bodies/Nerd/Skins/NerdBigShared/MAT_NerdBigAttachment
graph {
 properties {
  bAutomaticallySetUsageInEditor=false
  bUseMaterialAttributes=true
  bUsedWithMeshDeformer=true
  bUsedWithSkeletalMesh=true
  blendMode=BLEND_Masked
  shadingModel=MSM_Unlit
  usedShadingModels=Unlit
 }
 nodes [
  n1 = MaterialFunctionCall() {materialFunction=/VehicleCosmetics/UniversalSkins/Endit/MF_Skin_Endit}
  MaterialFunctionCall!(In=SetMaterialAttributes!(
    MaterialAttributes
    Emissive Color=GetMaterialAttributes(None=@n1.Result) {attributeGetTypes [#1]}.Emissive Color
    Opacity Mask=TextureSample() {samplerType=SAMPLERTYPE_Masks, texture=&/T_NerdBigAttachment_Alpha}.R
   ) {attributeSetTypes [#1, #2]}) {materialFunction=/VehicleCosmetics/Materials/MaterialFunctions/MF_Vehicle_FNPostFX}
 ]
}
